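# Masked two-sided vine material for Asteria grind rail props, using SplineThicken to give strips their width and parallax UVs.
# Vertex color drives width and variation, with switches for object-scaled rock vines and distance-based thinning.
# Moss tints upward-facing surfaces with an optional height gradient from the world-transformed origin.
# Optional world position offset pushes vines from the player, while mask textures feed dithered near-camera fade and pixel depth offset.
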Material /Game/Environments/Asteria/Props/GrindRailVines/M_GrindRailVines
graph {
 properties {
  bTangentSpaceNormal=false
  bUsedWithInstancedStaticMeshes=true
  bUsedWithNanite=true
  bUsedWithSplineMeshes=true
  blendMode=BLEND_Masked
  pixelDepthOffsetMode=PDOM_Legacy
  twoSided=true
  usedShadingModels=Default Lit
 }
 nodes [
  n1 = ScalarParameter() {defaultValue=7.542, parameterName=Width}
  n2 = VertexColor()
  n3 = LinearInterpolate(A=ScalarParameter() {parameterName=Width_End}, B=@n1, Alpha=@n2.R)
  n4 = Min(A=MaterialFunctionCall() {materialFunction=&ObjectScale}.Scale X)
  n5 = Multiply(A=@n2.R, B=ScalarParameter() {defaultValue=12.592, parameterName=WidthVar})
  n6 = Subtract(A=@n2.G) {constB=0.5}
  n7 = ScalarParameter() {defaultValue=19.7428, parameterName=WidthVarRandom}
  n8 = Add(
   A=StaticSwitchParameter(True=Multiply(A=@n3, B=@n4), False=Add(A=@n1, B=@n5)) {parameterName=IsRockVines}
   B=Multiply(A=@n2.G, B=Multiply(A=@n6, B=@n7))
  )
  n9 = Divide(
   A=Add(A=PixelDepth().R, B=ScalarParameter() {defaultValue=-6500, group=DistanceFade, parameterName=DistanceFadeOffset})
   B=ScalarParameter() {defaultValue=500, group=DistanceFade, parameterName=DistanceFadeLength}
  )
  n10 = StaticSwitchParameter(True=LinearInterpolate(A=@n8, B=Constant(), Alpha=Saturate(None=@n9)), False=@n8) {group=DistanceFade, parameterName=UseDistancefade}
  n11 = WorldPosition()
  n12 = WorldPosition()
  n13 = CollectionParameter() {collection=&FortniteMaterialParameters, parameterName=PlayerPosition}
  n14 = ComponentMask(None=@n13) {b=true, g=true, r=true}
  n15 = Add(A=Constant3Vector() {constant=<rgba 0 0 0 0>}, B=Subtract(A=@n12.XYZ, B=@n14))
  n16 = Saturate(None=Divide(A=Distance(A=@n14, B=@n12.XYZ), B=ScalarParameter() {defaultValue=200, parameterName=PlayerDist}))
  n17 = Multiply(A=Multiply(A=Normalize(VectorInput=@n15), B=OneMinus(None=@n16)), B=ScalarParameter() {defaultValue=30, parameterName=PlayerPush})
  n18 = MaterialFunctionCall(
   WidthBase=@n10
   WidthTip=@n10
   WorldPosition=StaticSwitchParameter(True=Add(A=@n11.XYZ, B=@n17), False=@n11.XYZ) {parameterName=UsePlayerReaction}
   AngleCorrectedNormals=StaticBool() {value=true}
   AdditionalNormal=MaterialFunctionCall(
    Normal=TextureSample(UVs=@n18.UVs with Parallax) {samplerType=SAMPLERTYPE_Normal, texture=&/GrindRailVines_N}.RGB
    Flatness=ScalarParameter() {parameterName=Normal}
   ) {materialFunction=&FlattenNormal}.Result
  ) {materialFunction=&SplineThicken}
  n19 = TextureSample(UVs=@n18.UVs with Parallax) {texture=&/GrindRailVines_basecolor}
  n20 = VectorParameter() {parameterName=ColorMoss, defaultValue=<rgba 0.3903 0.7292 0.262 1>}
  n21 = TransformPosition(None=Constant3Vector()) {transformType=TRANSFORMPOSSOURCE_World}
  n22 = Divide(A=Subtract(A=WorldPosition().Z, B=ComponentMask(None=@n21) {b=true}), B=ScalarParameter() {defaultValue=300, parameterName=GradientDist})
  n23 = Constant3Vector() {constant=<rgba 0 0 1 0>}
  n24 = Saturate(None=Power(Base=DotProduct(A=@n23, B=@n18.Normal), Exp=ScalarParameter() {defaultValue=5.7493, parameterName=FalloffMoss}))
  n25 = Add(A=OneMinus(None=Saturate(None=@n22)), B=@n24)
  LinearInterpolate(
   A=@n19.RGB
   B=LinearInterpolate(A=@n20.RGB, B=Multiply(A=@n20.RGB, B=@n19.R), Alpha=ScalarParameter() {defaultValue=1, parameterName=MossTextureAmount})
   Alpha=StaticSwitchParameter(True=Saturate(None=@n25), False=@n24) {parameterName=UseGroundGradient}
  )
  Multiply(A=TextureCoordinate(), B=Constant2Vector() {g=1, r=1})
  n26 = TextureSample(UVs=@n18.UVs with Parallax) {texture=&/GrindRailVines_M}
  n27 = Divide(
   A=Add(A=PixelDepth().R, B=ScalarParameter() {defaultValue=-100, parameterName=Camera Distance fadeOffset})
   B=ScalarParameter() {defaultValue=100, parameterName=Camera Distance fade}
  )
  Multiply(
   A=@n26.G
   B=MaterialFunctionCall(Alpha Threshold=Saturate(None=@n27)) {materialFunction=&DitherTemporalAA}.Result
  )
  Multiply(A=OneMinus(None=@n26.R), B=ScalarParameter() {defaultValue=1, parameterName=PDO})
  ScalarParameter() {defaultValue=0.9283, parameterName=Roughness}
  StaticSwitchParameter(True=Add(A=@n18.WorldPositionOffset, B=@n17), False=@n18.WorldPositionOffset) {parameterName=UsePlayerReaction}
 ]
}
