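# Opaque lit material for the 3D grind rail vine props in Asteria, set up for Nanite and spline meshes.
# Base color and normal textures share UVs with adjustable X tiling.
# The transformed normal's upward component, shaped by a falloff power, blends the base color toward a moss tint multiplied by its red channel.
# That red channel also drives clamped lerps for roughness and specular, avoiding separate mask textures.
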Material /Game/Environments/Asteria/Props/GrindRailVines/M_GrindRailVines_3D
graph {
 properties {bUsedWithNanite=true, bUsedWithSplineMeshes=true, pixelDepthOffsetMode=PDOM_Legacy, usedShadingModels=Default Lit}
 nodes [
  n1 = Multiply(A=TextureCoordinate(), B=AppendVector(A=ScalarParameter() {defaultValue=1, parameterName=TileX}, B=Constant() {r=1}))
  n2 = TextureSample(UVs=@n1) {texture=&/GrindRailVines_3D_basecolor}
  n3 = DotProduct(
   A=Constant3Vector() {constant=<rgba 0 0 1 0>}
   B=Transform(None=TextureSample(UVs=@n1) {samplerType=SAMPLERTYPE_Normal, texture=&/GrindRailVines_3D_N}.RGB)
  )
  n4 = ScalarParameter() {defaultValue=1.2502, parameterName=FalloffMoss}
  LinearInterpolate(
   A=@n2.RGB
   B=Multiply(A=VectorParameter() {parameterName=ColorMoss, defaultValue=<rgba 0.3903 0.7292 0.262 1>}.RGB, B=@n2.R)
   Alpha=Saturate(None=Power(Base=@n3, Exp=@n4))
  )
  Saturate(None=LinearInterpolate(A=ScalarParameter() {defaultValue=1.0992, parameterName=Roughness}, B=ScalarParameter() {defaultValue=-2.0378, parameterName=2}, Alpha=@n2.R))
  Saturate(None=LinearInterpolate(
    A=ScalarParameter() {defaultValue=0.068, parameterName=Specular}
    B=ScalarParameter() {defaultValue=0.7685, parameterName=Specular2}
    Alpha=@n2.R
   ))
 ]
}
