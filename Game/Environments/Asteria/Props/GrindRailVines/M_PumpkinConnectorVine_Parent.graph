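# Parent material for pumpkin grind rail connector vines, usable on spline meshes, blending mossy ground over a vine surface with a colored glow.
# The Is3xGrindrail switch selects 3D pumpkin vine or jungle rail shading, and moss placement by transformed-normal Z or UV sphere mask.
# World noise and HeightLerp shape the moss mask that blends color, normal, roughness and specular.
# Emissive combines a local-axis gradient, vertex-color mask, UV border bands and time-scrolled texture.
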Material /Game/Environments/Asteria/Props/GrindRailVines/M_PumpkinConnectorVine_Parent
graph {
 properties {bUsedWithSplineMeshes=true, pixelDepthOffsetMode=PDOM_Legacy, usedShadingModels=Default Lit}
 nodes [
  Constant()
  n1 = TextureCoordinate()
  Multiply(
   A=Add(A=ComponentMask(None=@n1) {g=true}, B=ScalarParameter() {defaultValue=-0.5, group=VineConector, parameterName=Emissive_Gradient Offset UV})
   B=ScalarParameter() {defaultValue=2, group=VineConector, parameterName=Emisssive_Gradient Size UV}
  )
  n2 = MaterialFunctionCall(Float2=TextureCoordinate()) {materialFunction=&BreakOutFloat2Components}
  n3 = ScalarParameter() {defaultValue=1, group=VineConector, parameterName=Tiling_U}
  n4 = ScalarParameter() {defaultValue=0.3, group=VineConector, parameterName=Tiling_V_Mult, primitiveDataIndex=18}
  NamedRerouteDeclaration(None=AppendVector(A=Multiply(A=@n2.R, B=@n3), B=Multiply(A=@n2.G, B=@n4))) {name=TexCocord, nodeColor=<rgba 1 0 0.0703 1>}
  n5 = MaterialFunctionCall(Float2=TextureCoordinate()) {materialFunction=&BreakOutFloat2Components}
  n6 = ScalarParameter() {defaultValue=34, group=VineConector, parameterName=Emissive_Boder UV Tile X}
  n7 = Add(A=Fmod(A=Multiply(A=@n5.R, B=@n6), B=Constant() {r=1})) {constB=-0.5}
  n8 = OneMinus(None=Abs(None=Multiply(A=@n7) {constB=2}))
  n9 = StaticComponentMaskParameter(None=VertexInterpolator(VS=LocalPosition() {localOrigin=InstancePreSkinning}.XYZ).PS) {defaultR=true, group=VineConector, parameterName=Emissive_Gradient Axis}
  n10 = Divide(
   A=StaticSwitchParameter(True=OneMinus(None=@n9), False=@n9) {defaultValue=true, group=VineConector, parameterName=Emissive_Gradient Invert}
   B=ScalarParameter() {defaultValue=3000, desc=500, group=VineConector, parameterName=Emissive_GradientSize}
  )
  n11 = ScalarParameter() {defaultValue=0.3, desc=0.5, group=VineConector, parameterName=Emissive_Gradient Offset, primitiveDataIndex=1}
  n12 = StaticComponentMaskParameter(None=VertexColor()) {defaultG=true, group=VineConector, parameterName=Emissive_VertexMask}
  n13 = StaticSwitchParameter(True=OneMinus(None=@n12), False=@n12) {group=VineConector, parameterName=Emissive_vtx mask Invert}
  n14 = ScalarParameter() {defaultValue=0.3, desc=0.5, group=VineConector, parameterName=Emissive_vtx mult, primitiveDataIndex=1}
  n15 = Saturate(None=Subtract(A=Add(A=@n10, B=@n11), B=Multiply(A=@n13, B=@n14)))
  n16 = Multiply(A=@n15, B=Multiply(A=@n8, B=@n8))
  n17 = ScalarParameter() {defaultValue=1, group=VineConector, parameterName=Emissive_Boder UV Mult}
  n18 = ScalarParameter() {defaultValue=0.165, group=VineConector, parameterName=Emissive_Pattern Mult}
  n19 = Subtract(A=Add(A=@n15, B=Multiply(A=@n16, B=@n17)), B=@n18)
  n20 = Multiply(A=@n5.R, B=ScalarParameter() {defaultValue=7, group=VineConector, parameterName=Emissive_Pattern Tile X})
  n21 = Add(
   A=Multiply(A=@n5.G, B=ScalarParameter() {defaultValue=1, group=VineConector, parameterName=Emissive_Pattern Tile Y})
   B=Divide(A=Time(), B=ScalarParameter() {defaultValue=-20, group=VineConector, parameterName=Emissive_Speed})
  )
  n22 = Multiply(
   A=TextureSample(UVs=AppendVector(A=@n20, B=@n21)) {texture=/Game/Environments/Helios/Props/Holiday/Holiday_LumpkinKing_A/Textures/T_LumpkinKing_Vine_Tiling_AOER}.G
   B=@n18
  )
  NamedRerouteDeclaration(None=Multiply(A=Add(A=@n19, B=@n22), B=VectorParameter() {group=VineConector, parameterName=Emissive_Color, defaultValue=<rgba 0 1.6667 5 1>}.RGB)) {name=Emissive, nodeColor=<rgba 0 1 0.7422 1>}
  n23 = MaterialFunctionCall(Float2=TextureCoordinate()) {materialFunction=&BreakOutFloat2Components}
  n24 = ScalarParameter() {defaultValue=1, group=VineConector, parameterName=Tiling_U_Moss}
  n25 = ScalarParameter() {defaultValue=1, group=VineConector, parameterName=Tiling_V_MossMult, primitiveDataIndex=18}
  NamedRerouteDeclaration(None=AppendVector(A=Multiply(A=@n23.R, B=@n24), B=Multiply(A=@n23.G, B=@n25))) {name=Texcord_Moss, nodeColor=<rgba 0 1 0.0625 1>}
  n26 = TextureSample(UVs=NamedRerouteUsage()) {samplerType=SAMPLERTYPE_Normal, texture=&/GrindRailVines_3D_N}
  n27 = TextureSample(UVs=NamedRerouteUsage()) {samplerType=SAMPLERTYPE_Normal, texture=/Game/Environments/Asteria/Sets/ZenGarden/Textures/T_Asteria_Zen_MossyGround_01_N}
  n28 = NamedRerouteUsage()
  n29 = TextureSample(UVs=NamedRerouteUsage()) {samplerType=SAMPLERTYPE_Normal, texture=&/T_JungleGrindRail_A_N}
  NamedRerouteDeclaration(None=StaticSwitchParameter(True=LinearInterpolate(A=@n26.RGB, B=@n27.RGB, Alpha=@n28), False=LinearInterpolate(A=@n29.RGB, B=@n27.RGB, Alpha=@n28)) {group=VineConector, parameterName=Is3xGrindrail}) {name=Normal, nodeColor=<rgba 0 0.7187 1 1>}
  n30 = NamedRerouteUsage()
  n31 = Saturate(None=LinearInterpolate(
    A=ScalarParameter() {defaultValue=1.0992, group=VineConector, parameterName=Roughness}
    B=ScalarParameter() {defaultValue=-2.0378, group=VineConector, parameterName=Roughness2}
    Alpha=@n30
   ))
  n32 = TextureSample(UVs=NamedRerouteUsage()) {samplerType=SAMPLERTYPE_Masks, texture=&/T_JungleGrindRail_A_S}
  NamedRerouteDeclaration(None=LinearInterpolate(
    A=StaticSwitchParameter(True=@n31, False=@n32.B) {group=VineConector, parameterName=Is3xGrindrail}
    B=ScalarParameter() {defaultValue=1, group=VineConector, parameterName=Moss_Rough, primitiveDataIndex=18}
    Alpha=NamedRerouteUsage()
   )) {name=Rough, nodeColor=<rgba 0 0.6953 1 1>}
  n33 = Saturate(None=LinearInterpolate(
    A=ScalarParameter() {defaultValue=0.068, group=VineConector, parameterName=Specular}
    B=ScalarParameter() {defaultValue=0.7685, group=VineConector, parameterName=Specular2}
    Alpha=@n30
   ))
  NamedRerouteDeclaration(None=LinearInterpolate(
    A=StaticSwitchParameter(True=@n33, False=@n32.R) {group=VineConector, parameterName=Is3xGrindrail}
    B=ScalarParameter() {group=VineConector, parameterName=Moss_Spec, primitiveDataIndex=18}
    Alpha=NamedRerouteUsage()
   )) {name=Spec, nodeColor=<rgba 0.0312 1 0 1>}
  n34 = TextureCoordinate()
  n35 = AppendVector(
   A=ScalarParameter() {defaultValue=1, group=VineConector, parameterName=Vine_TileX}
   B=ScalarParameter() {defaultValue=1, group=VineConector, parameterName=Vine_TileY}
  )
  n36 = TextureSample(UVs=NamedRerouteDeclaration(None=Multiply(A=@n34, B=@n35)) {name=Vine UV, nodeColor=<rgba 0 0.6953 1 1>}) {texture=&/GrindRailVines_3D_Pumpkin_basecolor}
  NamedRerouteDeclaration(None=@n36.R) {name=BaseColorRedChannel, nodeColor=<rgba 0 0.9531 1 1>}
  NamedRerouteDeclaration(None=@n26.RGB) {name=Normaltexture, nodeColor=<rgba 0 0.1797 1 1>}
  n37 = TextureSample(UVs=NamedRerouteUsage()) {samplerSource=SSM_Wrap_WorldGroupSettings, texture=&/T_JungleGrindRail_A_D}
  n38 = LinearInterpolate(A=@n36.RGB, B=Multiply(A=@n37.R, B=@n36.RGB), Alpha=@n36.RGB)
  n39 = Saturate(None=Add(A=NamedRerouteUsage()) {constB=0.7})
  n40 = NamedRerouteUsage()
  n41 = Multiply(A=Desaturation(
    None=TextureSample(UVs=@n40) {texture=/Game/Environments/Asteria/Sets/ZenGarden/Textures/T_Asteria_Zen_MossyGround_Pumpkin}.RGB
    Fraction=Constant() {r=0.2}
   ))
  NamedRerouteDeclaration(None=LinearInterpolate(
    A=StaticSwitchParameter(True=@n38, False=@n37.RGB) {group=VineConector, parameterName=Is3xGrindrail}
    B=Multiply(A=@n39, B=@n41)
    Alpha=NamedRerouteUsage()
   )) {name=BaseColor, nodeColor=<rgba 0 0.0625 1 1>}
  n42 = WorldPosition() {worldPositionShaderOffset=WPT_ExcludeAllShaderOffsets}
  n43 = ScalarParameter() {defaultValue=300, group=VineConector, parameterName=WorldNoise_Size}
  n44 = TextureSampleParameter2D(UVs=ComponentMask(None=Divide(A=@n42.XYZ, B=@n43)) {b=true, g=true}) {
   group=VineConector
   parameterName=WorldNoise
   samplerType=SAMPLERTYPE_LinearGrayscale
   texture=/Game/Global/Textures/Default/Noise/T_SwirlNoise_N_height
  }
  NamedRerouteDeclaration(None=@n44.R) {name=WorldNoise, nodeColor=<rgba 1 0 0.9609 1>}
  n45 = NamedRerouteUsage()
  n46 = OneMinus(None=ScalarParameter() {defaultValue=0.3, group=VineConector, parameterName=WorldNoise_Moss})
  n47 = Transform(None=NamedRerouteUsage())
  n48 = Saturate(None=LinearInterpolate(
    A=ScalarParameter() {defaultValue=0.64, group=VineConector, parameterName=Moss_Contrast_Low}
    B=ScalarParameter() {defaultValue=0.9904, group=VineConector, parameterName=Moss_Contrast_high}
    Alpha=ComponentMask(None=@n47) {b=true}
   ))
  n49 = TextureSample(UVs=Divide(A=NamedRerouteUsage(), B=ScalarParameter() {defaultValue=1, group=VineConector, parameterName=Moss_Var_Size})) {texture=/Game/Environments/Asteria/Sets/ZenGarden/Textures/T_Asteria_Zen_MossyGround_01_H}
  n50 = LinearInterpolate(
   A=ScalarParameter() {defaultValue=0.7697, group=VineConector, parameterName=Moss_Var_min}
   B=ScalarParameter() {defaultValue=4.1391, group=VineConector, parameterName=Moss_Var_max}
   Alpha=@n49.R
  )
  n51 = Subtract(A=MaterialFunctionCall(Float2=TextureCoordinate()) {materialFunction=&BreakOutFloat2Components}.R) {constB=0.5}
  n52 = LinearInterpolate(
   A=TextureSample(UVs=NamedRerouteUsage()) {samplerType=SAMPLERTYPE_LinearColor, texture=&/T_JungleGrindRail_A_H}.R
   Alpha=SphereMask(
    A=Abs(None=@n51)
    B=ScalarParameter() {defaultValue=0.5, group=VineConector, parameterName=MossUVPosition}
    Radius=ScalarParameter() {defaultValue=0.65, group=VineConector, parameterName=MossSpread}
   ) {hardnessPercent=0}
  )
  n53 = ScalarParameter() {defaultValue=1, group=VineConector, parameterName=MossSpread_Power}
  n54 = Multiply(
   A=Saturate(None=Add(A=@n45, B=@n46))
   B=StaticSwitchParameter(True=Multiply(A=@n48, B=@n50), False=Power(Base=@n52, Exp=@n53) {constExponent=0.6}) {group=VineConector, parameterName=Is3xGrindrail}
  )
  n55 = NamedRerouteDeclaration(None=Power(Base=@n49.R) {constExponent=1}) {name=Moss_Height, nodeColor=<rgba 0 1 0.7891 1>}
  n56 = ScalarParameter() {defaultValue=2, group=VineConector, parameterName=MossSpread_Contrast}
  NamedRerouteDeclaration(None=Multiply(A=MaterialFunctionCall(Transition Phase=@n54, Height Texture=@n55, Contrast=@n56) {materialFunction=&HeightLerp}.Alpha)) {name=MossMask, nodeColor=<rgba 0.0312 0 1 1>}
  NamedRerouteUsage()
  NamedRerouteUsage()
  NamedRerouteUsage()
  NamedRerouteUsage()
  NamedRerouteUsage()
 ]
}
